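# Physics asset for the WolfHunter axe pickaxe mesh, a two-handed melee cosmetic weapon, giving it a simple collision shape.
# It has a single capsule body on the Melee_TwoHanded_Handle bone, offset up the handle and pitched slightly with a yaw rotation.
# There are no constraints, so the weapon acts as one rigid body with no articulated parts.
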PhysicsAsset /BRCosmetics/Weapons/FORT_Melee/Pickaxe_WolfHunter/Meshes/WolfHunter_Axe_PhysicsAsset
bodies {Melee_TwoHanded_Handle [{shapeType=Capsule, length=65.2616, radius=30.9099, center=<-0.0146 2.1633 38.936>, rotation=<rot -0.8399 -90 0.0069>}]}
constraints []
